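; Environmental zipline actor strung between two placed poles that players ride in Fortnite Athena.
; The construction script fits the line mesh and overlap/interact capsules between pole sockets; BeginPlay initializes movement endpoints and removes the zipline if curve-table disabled.
; Cosmetic-side overlaps send enter/exit gameplay events to pawns, deferring walking players' entry until their movement mode changes, unless interact entry is used.
; Helper functions check for nearby buildings and update quest progress.
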(event EventBeginPlay
  (bind _poleasocketlocation (Variables|Default|GetPoleASocketLocation))
  (bind _polebsocketlocation (Variables|Default|GetPoleBSocketLocation))
  (bind _lineextendbeyondpolelength (Variables|Default|GetLineExtendBeyondPoleLength))
  (Variables|FortAthenaZipline|SetInteractIconPlacementCapsule (Variables|Default|GetOverlapCapsule))
  (Variables|Default|SetStartLocation (+ _poleasocketlocation (* (Math|Vector|GetUnitDirection(Vector) _poleasocketlocation _polebsocketlocation) _lineextendbeyondpolelength)))
  ;[0 0 0] Zip line movement will not work without being initialized
  (Variables|Default|SetTargetLocation (+ _polebsocketlocation (* (Math|Vector|GetUnitDirection(Vector) _polebsocketlocation _poleasocketlocation) _lineextendbeyondpolelength)))
  (Zipline|Initialize (Variables|Default|GetStartLocation) (Variables|Default|GetTargetLocation))
  (CallFunction|SetupOverlapBindings)
  (bind _asfort_game_mode_athena (Utilities|Casting|CastToFortGameModeAthena (Game|GetGameMode)))
  (if (MapInfo|IsMapInfoInitialized _asfort_game_mode_athena)
    (CallFunction|HandleonMapInfoInitialized)
    (else
      (MapInfo|BindEventtoOnMapInfoInitialized _asfort_game_mode_athena (delegate Custom|HandleOnMapInfoInitialized)))))

(event Custom|HandleOnMapInfoInitialized
  (bind _outxy (Fortnite|EvaluateCurveTableRow (Variables|Default|GetZipLineEnabledCTHandle)))
  (if (== (Math|Float|Truncate _outxy) 0)
    (Actor|DestroyActor (Variables|Default|GetPoleA))
    (Actor|DestroyActor self)))

(event Custom|HandleOnMovementModeChanged
  (params (Character :type Character%20Object%20Reference) (PrevMovementMode :type EMovementMode%20Enum :default "MOVE_None") (PreviousCustomMode :type Byte))
  (bind _asfort_player_pawn (Utilities|Casting|CastToFortPlayerPawn Character))
  (if (not (Pawn|Components|CharacterMovement|IsWalking (Variables|/Script/Engine.Character|GetCharacterMovement _asfort_player_pawn)))
    (bind _output_get (Variables|Default|SetPlayerPawn _asfort_player_pawn))
    (Character|UnbindEventfromMovementModeChangedDelegate _output_get (delegate Custom|HandleOnMovementModeChanged))
    (Utilities|Array|RemoveItem (Variables|Default|GetPlayersPendingMovementModeChange) (Variables|Default|GetPlayerPawn))
    (CallFunction|HandleEnterZipline (Variables|Default|GetPlayerPawn))))

(event Custom|HandleOnBeginOverlap
  (params (OverlappedComponent :type Primitive%20Component%20Object%20Reference) (OtherActor :type Actor%20Object%20Reference) (OtherComp :type Primitive%20Component%20Object%20Reference) (OtherBodyIndex :type Integer) (bFromSweep :type Boolean) (SweepResult :type Hit%20Result%20Structure%20%28by%20ref%29))
  (bind _playerpawn (Variables|Default|GetPlayerPawn))
  (Networking|CanExecuteCosmeticEvents
    (:True
      (if (not (Zipline|UseInteracttoEnterZipline))
        (bind _asfort_player_pawn (Utilities|Casting|CastToFortPlayerPawn OtherActor))
        (bind _output_get (Variables|Default|SetPlayerPawn _asfort_player_pawn))
        (if (Pawn|Components|CharacterMovement|IsWalking (Variables|/Script/Engine.Character|GetCharacterMovement _output_get))
          (Utilities|Array|AddUnique (Variables|Default|GetPlayersPendingMovementModeChange) (Variables|Default|GetPlayerPawn))
          (Character|BindEventtoMovementModeChangedDelegate (Variables|Default|GetPlayerPawn) (EventDispatchers|CreateEvent :!event Custom|HandleOnMovementModeChanged))
          (elif (Zipline|HasRecentlyExitedZiplinefromJumping _playerpawn)
            _
            (elif (not (Health|IsDBNO _playerpawn))
              (CallFunction|HandleEnterZipline (Variables|Default|GetPlayerPawn)))))))))

(event Custom|HandleOnEndOverlap
  (params (OverlappedComponent :type Primitive%20Component%20Object%20Reference) (OtherActor :type Actor%20Object%20Reference) (OtherComp :type Primitive%20Component%20Object%20Reference) (OtherBodyIndex :type Integer))
  (Networking|CanExecuteCosmeticEvents
    (:True
      (bind _asfort_player_pawn (Utilities|Casting|CastToFortPlayerPawn OtherActor))
      (bind _output_get (Variables|Default|SetPlayerPawn _asfort_player_pawn))
      (Ability|SendGameplayEventtoActor _output_get "(TagName=\"Abilities.Generic.Athena.ZipLine.Exit\")" (Utilities|Struct|MakeGameplayEventData "(TagName=\"Abilities.Generic.Athena.ZipLine.Exit\")"))
      (Character|UnbindEventfromMovementModeChangedDelegate (Variables|Default|GetPlayerPawn) (EventDispatchers|CreateEvent :!event Custom|HandleOnMovementModeChanged))
      (Utilities|Array|RemoveItem (Variables|Default|GetPlayersPendingMovementModeChange) (Variables|Default|GetPlayerPawn)))))

(fn CheckForBuildings
  (params (Player :type Fort%20Player%20Pawn%20Object%20Reference))
  (bind _outxy (Fortnite|EvaluateCurveTableRow (Variables|Default|GetBuildingCheckBoxExtentsCTHandle)))
  (bind _outactors (Collision|BoxOverlapActors (Math|Vector|MakeVector (.x (- (Math|Vector|FindClosestPointonSegment (Transformation|GetActorLocation Player) (Variables|Default|GetStartLocation) (Variables|Default|GetTargetLocation)) (Variables|Default|GetPlayerOffsetToLine))) (.y (- (Math|Vector|FindClosestPointonSegment (Transformation|GetActorLocation Player) (Variables|Default|GetStartLocation) (Variables|Default|GetTargetLocation)) (Variables|Default|GetPlayerOffsetToLine))) (+ (.z (- (Math|Vector|FindClosestPointonSegment (Transformation|GetActorLocation Player) (Variables|Default|GetStartLocation) (Variables|Default|GetTargetLocation)) (Variables|Default|GetPlayerOffsetToLine))) 75.0)) (Math|Vector|MakeVector _outxy _outxy _outxy) (Utilities|Array|MakeArray "ObjectTypeQuery1") 0 (Utilities|Array|MakeArray Player) :!class /Script/Engine.KismetSystemLibrary))
  (bind _array_element (Utilities|Array|ForEachLoopwithBreak _outactors)
    (:LoopBody
      (bind _asbuilding_actor (Utilities|Casting|CastToBuildingActor _array_element))
      (Utilities|Casting|CastToBuildingProp _asbuilding_actor
        (:CastFailed
          (LocalVariables|SetBuildingsNearby true))))
    (:Completed
      _))
  (return (LocalVariables|GetBuildingsNearby)))

(fn GetCharacterCustomMovementMode
  (params (Character :type Pawn%20Object%20Reference))
  (return (Utilities|Enum|BytetoEnumEFortCustomMovement (Variables|/Script/Engine.CharacterMovementComponent|GetCustomMovementMode (Utilities|Casting|CastToCharacterMovementComponent (Pawn|GetMovementComponent Character)))) (Variables|/Script/Engine.CharacterMovementComponent|GetMovementMode (Utilities|Casting|CastToCharacterMovementComponent (Pawn|GetMovementComponent Character)))))

(fn HandleEnterZipline
  (params (InPlayerPawn :type Fort%20Player%20Pawn%20Object%20Reference))
  (Ability|SendGameplayEventtoActor (Variables|GetInPlayerPawn) "(TagName=\"Abilities.Generic.Athena.ZipLine.Enter\")" (Utilities|Struct|MakeGameplayEventData "(TagName=\"Abilities.Generic.Athena.ZipLine.Enter\")" self)))

(fn QuestCheckForUse
  (params (FortPlayerController :type Fort%20Player%20Controller%20Object%20Reference))
  (bind _backendname (Variables|QuestStuff|GetBackendName))
  (bind _questitem (Variables|QuestStuff|GetQuestItem))
  (bind _cancompletequest (Class|AthenaQuestFunctionLibrary|AthenaCheckCreativeMode (Game|GetGameState)))
  (if _cancompletequest
    ;[1 1 1] Quest Check
    (Development|PrintString "Check For Progress" false false "(R=0.000000,G=0.660000,B=1.000000,A=1.000000)" 5.0)
    (Utilities|FlowControl|Sequence
      (:then_0
        (Networking|SwitchHasAuthority
          (:Authority
            (bind (_questisvalid _questcompleted _objectivecompleted) (Default|AthenaCheckQuestandObjectiveCompletion FortPlayerController _questitem _backendname))
            (if (and _questisvalid (not _objectivecompleted))
              (Class|AthenaQuestFunctionLibrary|AthenaGiveQuestUpdatetoPlayers _questitem _backendname (Variables|QuestStuff|GetObjStat) (Utilities|Array|MakeArray (Fortnite|GetFortPlayerControllerfromActor FortPlayerController)))))))
      (:then_1
        (return)))))

(fn SetupOverlapBindings
  (Networking|CanExecuteCosmeticEvents
    (:True
      (Variables|/Script/Engine.PrimitiveComponent|SetbGenerateOverlapEvents true (Variables|Default|GetOverlapCapsule))
      (Collision|BindEventtoOnComponentBeginOverlap (Variables|Default|GetOverlapCapsule) (EventDispatchers|CreateEvent :!event HandleOnBeginOverlap))
      (Collision|BindEventtoOnComponentEndOverlap (Variables|Default|GetOverlapCapsule) (EventDispatchers|CreateEvent :!event HandleOnEndOverlap)))))

(fn StartZipLining
  (params (FortPlayerPawn :type Fort%20Player%20Pawn%20Object%20Reference))
  (Networking|ForceNetUpdate FortPlayerPawn))

(fn ConstructionScript
  (bind _poleasocketlocation (Variables|Default|GetPoleASocketLocation))
  (bind _polea (Variables|Default|GetPoleA))
  (bind _poleb (Variables|Default|GetPoleB))
  (bind _polebsocketlocation (Variables|Default|GetPoleBSocketLocation))
  (bind _lineextendbeyondpolelength (Variables|Default|GetLineExtendBeyondPoleLength))
  (Utilities|IsValid _polea
    (:Is%20Valid
      (Variables|Default|SetPoleASocketLocation (Transformation|GetSocketLocation (Variables|/Script/Engine.Actor|GetRootComponent _polea) "LineAttachSocket")))
    (:Is%20Not%20Valid
      _))
  (Utilities|IsValid _poleb
    (:Is%20Valid
      (Variables|Default|SetPoleBSocketLocation (Transformation|GetSocketLocation (Variables|/Script/Engine.Actor|GetRootComponent _poleb) "LineAttachSocket")))
    (:Is%20Not%20Valid
      _))
  (if (and (Utilities|IsValid _poleb :!class /Script/Engine.KismetSystemLibrary) (Utilities|IsValid _polea :!class /Script/Engine.KismetSystemLibrary))
    (Variables|Default|SetStartLocation (+ _poleasocketlocation (* (Math|Vector|GetUnitDirection(Vector) _poleasocketlocation _polebsocketlocation) _lineextendbeyondpolelength)))
    (Variables|Default|SetTargetLocation (+ _polebsocketlocation (* (Math|Vector|GetUnitDirection(Vector) _polebsocketlocation _poleasocketlocation) _lineextendbeyondpolelength)))
    (Transformation|SetActorLocation (/ (+ (Variables|Default|GetPoleASocketLocation) (Variables|Default|GetPoleBSocketLocation)) 2))
    (Transformation|SetWorldRotation (Variables|Default|GetZipline) (Math|Rotator|MakeRotator (+ (.pitch (Math|Rotator|FindLookatRotation (Variables|Default|GetPoleASocketLocation) (Variables|Default|GetPoleBSocketLocation))) 9e+01) (.roll (Math|Rotator|FindLookatRotation (Variables|Default|GetPoleASocketLocation) (Variables|Default|GetPoleBSocketLocation))) (+ (.yaw (Math|Rotator|FindLookatRotation (Variables|Default|GetPoleASocketLocation) (Variables|Default|GetPoleBSocketLocation))) 9e+01)) false false)
    (bind _distance (Class|MissionBlueprintFunctionLibrary|DistanceBetweenTwoVectors (Variables|Default|GetPoleASocketLocation) (Variables|Default|GetPoleBSocketLocation)))
    (Components|Capsule|SetCapsuleHalfHeight (Variables|Default|GetOverlapCapsule) (/ _distance 2.0) true)
    (Components|Capsule|SetCapsuleHalfHeight (Variables|Default|GetInteractCapsule) (/ _distance 2.02) true)
    (bind _outxy (Fortnite|EvaluateCurveTableRow (Variables|Default|GetLineCollisionRadiusCTHandle)))
    (Components|Capsule|SetCapsuleRadius (Variables|Default|GetOverlapCapsule) _outxy true)
    (Components|Capsule|SetCapsuleRadius (Variables|Default|GetInteractCapsule) (* _outxy 2.0) true)
    (Transformation|AddWorldOffset (Variables|Default|GetOverlapCapsule) (Math|Vector|MakeVector 0.0 0.0 (+ (* -1 _outxy) (Variables|Default|GetCollisionHeightAboveLine))) false false)
    (Transformation|AddWorldOffset (Variables|Default|GetInteractCapsule) (Math|Vector|MakeVector 0.0 0.0 -1.5e+02) false false)
    (Transformation|SetWorldScale3D (Variables|Default|GetSM_Line) (Math|Vector|MakeVector (Variables|Default|GetLineThickness) (Variables|Default|GetLineThickness) (/ _distance 1e+02)))))
