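# Translucent unlit material for the zipline consumable, with responsive AA enabled.
# One color branch multiplies the base color by a time-of-day tint and a vertical gradient from vertex normals.
# Another blends toward a dark color using camera-distance falloffs, including a sharp close-range term.
# A vertex-normal vector scales from Width to a larger value with camera distance, a technique that keeps thin cable geometry readable far away.
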Material /Figment_S10/Athena/Items/Consumables/ZipLine/M_Zipline
graph {
 properties {bEnableResponsiveAA=true, blendMode=BLEND_Translucent, pixelDepthOffsetMode=PDOM_Legacy, shadingModel=MSM_Unlit, usedShadingModels=Unlit}
 nodes [
  n1 = VertexNormalWS()
  n2 = VectorParameter() {defaultValue=<rgba 0 0 1 1>}
  n3 = Clamp(None=ConstantBiasScale(None=DotProduct(A=@n1, B=@n2.RGB)))
  Add(A=Multiply(A=@n3, B=ScalarParameter() {defaultValue=1, parameterName=Param_5}), B=ScalarParameter() {defaultValue=0.1476})
  n4 = VectorParameter() {parameterName=Color, defaultValue=<rgba 0.1712 0.2035 0.2448 1>}
  n5 = MaterialFunctionCall(Vector 3=Subtract(A=CameraPositionWS(), B=WorldPosition().XYZ)) {materialFunction=/Engine/Functions/Engine_MaterialFunctions02/Utility/VectorLength}
  n6 = ScalarParameter() {defaultValue=-3613.9133, parameterName=Param_2}
  n7 = Clamp(None=Divide(A=Add(A=@n5.V3 Length, B=@n6), B=ScalarParameter() {defaultValue=18265.3379, parameterName=NormalizeDistance})) {maxDefault=100000000}
  n8 = OneMinus(None=@n7)
  n9 = ScalarParameter() {defaultValue=0.25, parameterName=Param_3}
  n10 = WorldPosition()
  n11 = CameraPositionWS()
  n12 = Divide(
   A=Add(A=Distance(A=@n10.XYZ, B=@n11), B=ScalarParameter() {defaultValue=-15, parameterName=Param_9})
   B=ScalarParameter() {defaultValue=18.9689, parameterName=Param_8}
  )
  n13 = Power(Base=OneMinus(None=Clamp(None=@n12)), Exp=ScalarParameter() {defaultValue=6, parameterName=Param_7})
  LinearInterpolate(
   A=@n4.RGB
   B=VectorParameter() {parameterName=ColorDark, defaultValue=<rgba 0.2708 0.2708 0.2708 1>}.RGB
   Alpha=Multiply(A=Clamp(None=@n8, Min=@n9), B=OneMinus(None=@n13))
  )
  LinearInterpolate() {constA=0.5}
  n14 = VectorParameter() {parameterName=Morning Color, defaultValue=<rgba 0.9167 0.5537 0.7144 1>}
  n15 = VectorParameter() {parameterName=Day Color, defaultValue=<rgba 1 0.9958 0.9844 1>}
  n16 = VectorParameter() {parameterName=Evening Color, defaultValue=<rgba 0.7917 0.6073 0.6389 1>}
  n17 = VectorParameter() {parameterName=Night Color, defaultValue=<rgba 0.5426 0.6778 0.974 1>}
  Multiply(
   A=@n4.RGB
   B=Multiply(
    A=MaterialFunctionCall(Color_Morning=@n14.RGB, Color_Day=@n15.RGB, Color_Evening=@n16.RGB, Color_Night=@n17.RGB) {materialFunction=&Color_Time_of_Day}.Result
    B=LinearInterpolate(Alpha=@n3) {constA=0.25}
   )
  )
  Multiply(
   A=VertexNormalWS()
   B=LinearInterpolate(A=ScalarParameter() {defaultValue=2, parameterName=Width}, B=ScalarParameter() {defaultValue=20, parameterName=Param_1}, Alpha=@n7)
  )
  ScalarParameter() {defaultValue=1, parameterName=Spec}
  ScalarParameter() {defaultValue=0.5, parameterName=Rough}
 ]
}
